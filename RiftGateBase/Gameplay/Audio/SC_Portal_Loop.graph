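# Looping ambient SoundCue for a rift portal, driven by an OpenAmount parameter through continuous modulators.
# It mixes four attenuated layers: a far base bed, a medium destabilize loop, and close destabilize and open-portal loops.
# Medium and close layers gain pitch and volume as OpenAmount rises.
# In the far bed, the closed loop drops in pitch while two chained modulators ramp the open loop up early, then ease it down.
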SoundCue /RiftGateBase/Gameplay/Audio/SC_Portal_Loop
root=Mixer_0
nodes [
 {id=Attenuation_0, children [ModulatorContinuous_6], properties {attenuationSettings=&/SA_Portal_Loop_Close}}
 {id=Attenuation_1, children [ModulatorContinuous_5], properties {attenuationSettings=&/SA_Portal_Loop_Med}}
 {id=Attenuation_2, children [Mixer_2], properties {attenuationSettings=&/SA_Portal_Loop_Far}}
 {id=Attenuation_3, children [ModulatorContinuous_4], properties {attenuationSettings=&SA_Portal_Loop_Close}}
 {id=Mixer_0, children [null,Mixer_1], properties {inputVolume [1,1]}}
 {id=Mixer_1, children [Attenuation_2, Attenuation_1, Attenuation_0, Attenuation_3], properties {inputVolume [1,0.5,0.5,1.5]}}
 {id=Mixer_2, children [ModulatorContinuous_2, ModulatorContinuous_0], properties {inputVolume [1,1]}}
 {
  id=ModulatorContinuous_0
  children [WavePlayer_0]
  properties {pitchModulationParams {default=0, maxOutput=0.5, minOutput=1, parameterName=OpenAmount}}
 }
 {
  id=ModulatorContinuous_1
  children [WavePlayer_1]
  properties {
   pitchModulationParams {default=0, maxInput=0.1, minOutput=0.7, parameterName=OpenAmount}
   volumeModulationParams {default=0, maxInput=0.1, minOutput=0.2, parameterName=OpenAmount}
  }
 }
 {
  id=ModulatorContinuous_2
  children [ModulatorContinuous_1]
  properties {
   pitchModulationParams {default=0, maxOutput=0.8, minInput=0.5, minOutput=1, parameterName=OpenAmount}
   volumeModulationParams {default=0, maxOutput=0.8, minInput=0.5, minOutput=1, parameterName=OpenAmount}
  }
 }
 {
  id=ModulatorContinuous_4
  children [WavePlayer_3]
  properties {
   pitchModulationParams {default=0, maxOutput=1.25, minOutput=1, parameterName=OpenAmount}
   volumeModulationParams {default=0, minOutput=0.5, parameterName=OpenAmount}
  }
 }
 {
  id=ModulatorContinuous_5
  children [WavePlayer_5]
  properties {
   pitchModulationParams {default=0, maxOutput=1.25, minOutput=1, parameterName=OpenAmount}
   volumeModulationParams {default=0, minOutput=0.5, parameterName=OpenAmount}
  }
 }
 {
  id=ModulatorContinuous_6
  children [WavePlayer_6]
  properties {
   pitchModulationParams {default=0, maxOutput=1.25, minOutput=1, parameterName=OpenAmount}
   volumeModulationParams {default=0, minOutput=0.5, parameterName=OpenAmount}
  }
 }
 {id=WavePlayer_0, properties {bLooping=true, soundWaveAssetPtr=/Game/Athena/Sounds/Events/Pickle/Pickle_Amb_Base_Closed_Loop_01}}
 {id=WavePlayer_1, properties {bLooping=true, soundWaveAssetPtr=/Game/Athena/Sounds/Events/Pickle/Pickle_Amb_Open_Base_Loop_01}}
 {id=WavePlayer_3, properties {bLooping=true, soundWaveAssetPtr=&/Blondejaw_Portal_Open_Amb_Loop}}
 {id=WavePlayer_5, properties {bLooping=true, soundWaveAssetPtr=/Game/Athena/Sounds/Events/Pickle/Pickle_Amb_Destabilize_01_Loop}}
 {id=WavePlayer_6, properties {bLooping=true, soundWaveAssetPtr=/Game/Athena/Sounds/Events/Pickle/Pickle_Amb_Destabilize_02_Loop}}
]
properties {maxDistance=10000, soundClassObject=&Fort_BossFight_SFX, volumeMultiplier=0.7}
