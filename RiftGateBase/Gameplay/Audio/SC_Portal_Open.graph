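# Sound cue for a portal opening in RiftGateBase, built from close and distant layers.
# Each wave plays through its own attenuation asset, and a mixer blends the two, weighting the distant layer louder.
# The cue is assigned to the boss-fight player-critical sound class, with bus and submix sends disabled.
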SoundCue /RiftGateBase/Gameplay/Audio/SC_Portal_Open
root=Mixer_0
nodes [
 {id=Attenuation_0, children [WavePlayer_0], properties {attenuationSettings=&/SA_Portal_Close}}
 {id=Attenuation_1, children [WavePlayer_3], properties {attenuationSettings=&/SA_Portal_Distant}}
 {id=Mixer_0, children [Attenuation_0, Attenuation_1], properties {inputVolume [0.7,1.5]}}
 {id=WavePlayer_0, properties {soundWaveAssetPtr=&/Blondejaw_Portal_Open_Close_01}}
 {id=WavePlayer_3, properties {soundWaveAssetPtr=&/Blondejaw_Portal_Open_Distant_01}}
]
properties {
 bEnableBusSends=false
 bEnableSubmixSends=false
 maxDistance=150000
 soundClassObject=/Game/Sounds/FortSoundClassesAndModes/Fort_BossFight_SoundClass/Fort_BossFight_PlayerCritical
 volumeMultiplier=1
 attenuationOverrides {distanceAlgorithm=NaturalSound, falloffDistance=80000, lPFRadiusMin=17, stereoSpread=1000, attenuationShapeExtents {x=8000}}
 concurrencyOverrides {resolutionRule=PreventNew}
}
