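# Sound cue for the Rift Gate pulse knockback effect.
# A random node with equal weights picks one of three knockback wave variations on each play.
# It uses the close Rift Gate attenuation and the gadgets sound class, and overrides concurrency so only one instance plays at a time.
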SoundCue /RiftGateBase/Gameplay/Audio/SC_RiftGate_Pulse_Knockback
root=Random_2
nodes [
 {id=Random_2, children [WavePlayer_8, WavePlayer_9, WavePlayer_10], properties {weights [1,1,1]}}
 {id=WavePlayer_10, properties {soundWaveAssetPtr=&/Blondejaw_RiftGate_Pulse_Knockback_03}}
 {id=WavePlayer_8, properties {soundWaveAssetPtr=&/Blondejaw_RiftGate_Pulse_Knockback_01}}
 {id=WavePlayer_9, properties {soundWaveAssetPtr=&/Blondejaw_RiftGate_Pulse_Knockback_02}}
]
properties {
 attenuationSettings=&/SA_RiftGate_Close
 bOverrideConcurrency=true
 maxDistance=11000
 soundClassObject=&Fort_Gadgets
 volumeMultiplier=1
 concurrencyOverrides {maxCount=1}
}
