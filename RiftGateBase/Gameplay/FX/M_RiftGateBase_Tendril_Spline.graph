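# Translucent unlit material for rift gate tendril effects on Niagara mesh particles and spline meshes.
# Two panned Zero Point star-line textures add to a wave-tinted, masked background for emissive, scaled by particle color.
# Opacity combines soft masks with view-angle, camera and depth fades, replaced by a constant on forward and mobile.
# World position offset combines SplineThicken with dynamic-parameter-driven sine waves; a pull toward CenterPositionA is multiplied by zero.
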Material /RiftGateBase/Gameplay/FX/M_RiftGateBase_Tendril_Spline
graph {
 properties {
  bUsedWithNiagaraMeshParticles=true
  bUsedWithSplineMeshes=true
  blendMode=BLEND_Translucent
  pixelDepthOffsetMode=PDOM_Legacy
  shadingModel=MSM_Unlit
  twoSided=true
  usedShadingModels=Unlit
 }
 nodes [
  n1 = ParticleColor()
  n2 = VectorParameter() {parameterName=PanDirection1, defaultValue=<rgba 0.08 1 0 0>}
  n3 = AppendVector(A=@n2.R, B=@n2.R)
  n4 = Add(A=Multiply(A=Time(), B=ScalarParameter() {defaultValue=1, parameterName=PanSpeed}), B=ParticleRandom())
  n5 = VectorParameter() {parameterName=Tile1, defaultValue=<rgba 1 1 0 0>}
  n6 = ComponentMask(None=TextureCoordinate()) {g=true}
  n7 = SphereMask(A=Saturate(None=OneMinus(None=@n6)), B=Constant() {r=0.1}) {attenuationRadius=1, hardnessPercent=90}
  n8 = Ceil(None=@n7)
  n9 = ScalarParameter() {defaultValue=200, parameterName=SplineWidth}
  n10 = TextureCoordinate()
  n11 = Multiply(A=OneMinus(None=ComponentMask(None=@n10) {g=true}), B=ScalarParameter() {defaultValue=1, parameterName=WaveTiling})
  n12 = ScalarParameter() {defaultValue=2, parameterName=WaveTilingBias}
  n13 = ScalarParameter() {defaultValue=0.5, parameterName=WaveLength}
  n14 = DynamicParameter() {parameterIndex=1, paramNames [WaveLength,WaveScale,BendToZP,Param4]}
  n15 = Time()
  n16 = ParticleRandom()
  n17 = Add(
   A=Divide(A=Power(Base=@n11, Exp=@n12) {constExponent=0.5}, B=Multiply(A=@n13, B=@n14.WaveLength))
   B=Multiply(A=Add(A=@n15, B=@n16), B=ScalarParameter() {defaultValue=-0.5, parameterName=WaveSpeed})
  )
  n18 = ConstantBiasScale(None=Sine(None=@n17))
  n19 = LinearInterpolate(Alpha=@n18) {constA=0.4}
  n20 = Multiply(A=Multiply(A=Multiply(A=@n8, B=@n9), B=DynamicParameter() {paramNames [Random,Grow,Width,Param4]}.Width), B=@n19)
  n21 = Multiply(A=ScalarParameter() {defaultValue=1, parameterName=SplineWidthTileMult}, B=@n20) {constB=0.1}
  n22 = Multiply(A=TextureCoordinate(), B=AppendVector(A=Multiply(A=@n5.R, B=@n21), B=@n5.G))
  n23 = TextureSample(UVs=Panner(Coordinate=@n22, Time=@n4, Speed=@n3) {speedX=1, speedY=1}) {texture=/Game/Athena/Environments/Nexus/Textures/T_ZeroPoint_Line_Stars}
  n24 = VectorParameter() {parameterName=PanDirection1Mult, defaultValue=<rgba -1.7 1.4 0 1>}
  n25 = Multiply(A=@n3, B=AppendVector(A=@n24.R, B=@n24.R))
  n26 = Multiply(A=@n22) {constB=2}
  n27 = Add(
   A=@n23.RGB
   B=TextureSample(UVs=Panner(Coordinate=@n26, Time=@n4, Speed=@n25) {speedX=1, speedY=1}) {texture=&T_ZeroPoint_Line_Stars}.RGB
  )
  n28 = ComponentMask(None=TextureCoordinate()) {r=true}
  n29 = Constant() {r=0.5}
  n30 = Multiply(A=@n7, B=Multiply(A=SphereMask(A=@n28, B=@n29, Radius=@n29) {hardnessPercent=0}))
  n31 = MaterialFunctionCall(In=@n30, Contrast=Constant() {r=0.25}) {materialFunction=&CheapContrast}
  n32 = Multiply(
   A=LinearInterpolate(
    A=VectorParameter() {parameterName=BackgroundColor1, defaultValue=<rgba 1.8542 1.9028 2 0>}.RGB
    B=VectorParameter() {parameterName=BackgroundColor2, defaultValue=<rgba 0.2 0.2 0.2 0>}.RGB
    Alpha=@n18
   )
   B=@n31.Result
  )
  MaterialFunctionCall(Emissive=Multiply(A=Add(A=@n27, B=@n32), B=@n1.RGB)) {materialFunction=&MF_FX_EmissiveAdjustments}
  n33 = Multiply(A=@n31.Result) {constB=0.4}
  n34 = MaterialFunctionCall() {materialFunction=&LinearGradient}
  n35 = Multiply(
   A=Multiply(A=Add(A=@n23.B, B=@n33) {constB=0.5}, B=@n30)
   B=SphereMask(A=Power(Base=@n34.VGradient), B=Constant() {r=0.5}) {attenuationRadius=0.5, hardnessPercent=0}
  )
  n36 = Multiply(A=MaterialFunctionCall("FadeSensitivity (optional)"=ScalarParameter() {defaultValue=1, parameterName=AngleFade}) {materialFunction=&FadeBasedOnViewAngle}.Result) {constB=2}
  n37 = MaterialFunctionCall(Distance=ScalarParameter() {defaultValue=200, parameterName=NearFadeDIst}) {materialFunction=&Near_Camera_Fade}
  n38 = DepthFade(FadeDistance=ScalarParameter() {defaultValue=500, parameterName=DepthFade})
  n39 = Constant() {r=0.7}
  n40 = ShadingPathSwitch(Default=Multiply(A=Saturate(None=@n36), B=Multiply(A=@n37.Result, B=@n38)), Forward=@n39, Mobile=@n39)
  Multiply(A=@n1.A, B=Multiply(A=Multiply(A=@n35, B=@n40), B=@n19))
  n41 = ScalarParameter() {parameterName=WPO_Mask_Min}
  n42 = MaterialFunctionCall() {materialFunction=&LinearGradient}
  n43 = SphereMask(A=Power(Base=@n42.VGradient), B=Constant() {r=0.5}) {attenuationRadius=0.5, hardnessPercent=0}
  n44 = Multiply(A=OneMinus(None=@n42.VGradient), B=Power(Base=OneMinus(None=@n43)) {constExponent=5})
  n45 = Add(A=ActorPositionWS(), B=VectorParameter() {parameterName=CenterPositionA, defaultValue=<rgba 0 0 6000 1>}.RGB)
  n46 = ParticlePositionWS()
  n47 = Multiply(A=Multiply(A=LinearInterpolate(A=@n41, Alpha=@n44) {constA=0.2}, B=Subtract(A=@n45, B=@n46)), B=Constant())
  n48 = MaterialFunctionCall(WidthBase=@n20, WidthTip=@n20) {materialFunction=&SplineThicken}
  n49 = MaterialFunctionCall() {materialFunction=&TangentBasis}
  n50 = Multiply(A=LinearInterpolate(A=@n49.X U, B=@n49.X U, Alpha=@n18), B=Sine(None=@n17))
  n51 = Multiply(A=ScalarParameter() {defaultValue=50, parameterName=WaveScale}, B=@n14.WaveScale)
  n52 = LinearInterpolate(
   A=Divide(A=@n51) {constB=10}
   B=@n51
   Alpha=OneMinus(None=MaterialFunctionCall() {materialFunction=&LinearGradient}.VGradient)
  )
  Reroute(None=Add(A=Add(A=@n47, B=@n48.WorldPositionOffset), B=Multiply(A=@n50, B=@n52)))
 ]
}
