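# Translucent deferred decal for a Rift Gate base scorch mark, with grunge, burn and glowing blue-to-cyan lightning layers.
# One burn-mark texture supplies the layers by channel, animated by noise panned in radial coordinates, with time-of-day-adjusted glow.
# Center-scaled UVs collapse on non-horizontal surfaces via rounded world normal Z, except on mobile.
# Combined opacity uses SphereMask thresholds, then fades by vertical distance and decal lifetime.
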Material /RiftGateBase/Gameplay/Materials/M_RiftGateBase_Scorch_Decal
graph {
 properties {blendMode=BLEND_Translucent, materialDomain=MD_DeferredDecal, pixelDepthOffsetMode=PDOM_Legacy, usedShadingModels=Default Lit}
 nodes [
  Constant() {r=0.005}
  n1 = Round(None=Abs(None=SceneTexture() {sceneTextureId=PPI_WorldNormal}.Color))
  n2 = LinearInterpolate(
   A=Constant()
   B=NamedRerouteUsage()
   Alpha=ShadingPathSwitch(Default=ComponentMask(None=@n1) {b=true}, Mobile=Constant() {r=1}) {desc=Migrated from FeatureLevelSwitch}
  ) {constAlpha=1}
  n3 = TextureSample(UVs=@n2) {samplerSource=SSM_Wrap_WorldGroupSettings, samplerType=SAMPLERTYPE_Masks, texture=/Game/Environments/Props/Decals/Textures/T_BurnMark_Decal}
  n4 = NamedRerouteUsage()
  n5 = Multiply(A=MaterialFunctionCall(Vector or UVs=Subtract(A=@n4) {constB=0.5}) {materialFunction=&VectorToRadialValue}.Radial Coordinates)
  n6 = TextureSample(UVs=Panner(Coordinate=@n5) {speedY=-0.075}) {samplerType=SAMPLERTYPE_Masks, texture=/Game/Weapons/WeaponSkins/Wraps/Textures/Mash/Mash_Noise}
  n7 = Multiply(A=Multiply(A=ScalarParameter() {defaultValue=1, parameterName=LightMark0-1}, B=@n3.G) {constA=1}, B=Saturate(None=Add(A=@n6.R) {constB=0.2}))
  n8 = Multiply(
   A=@n7
   B=MaterialFunctionCall!(Color=ScalarParameter() {defaultValue=8, parameterName=LightningEmissive}) {materialFunction=/Game/Packages/Fortress_SharedMaterials/Anime_Master_Material/MF_TODColorCurve}.Result
  ) {constB=3}
  n9 = Constant3Vector() {constant=<rgba 0.154 0.1501 0.7031 0>}
  n10 = Constant3Vector() {constant=<rgba 0.128 0.64 0.7031 0>}
  n11 = TextureSample() {
   samplerSource=SSM_Wrap_WorldGroupSettings
   samplerType=SAMPLERTYPE_LinearGrayscale
   texture=/Game/Weapons/FORT_Melee/Pickaxe_SnowFall_Female/FX/Snowfall_pick_Noise
  }
  Multiply(
   A=Multiply(A=Add(A=@n8) {constB=0}, B=LinearInterpolate(A=@n9, B=@n10, Alpha=@n11.R))
   B=LinearInterpolate(A=Constant() {r=0.2}, B=Constant() {r=1}, Alpha=Power(Base=@n6.R) {constExponent=1}) {constA=1, constB=0.1}
  )
  n12 = NamedRerouteUsage()
  n13 = WorldPosition()
  n14 = ObjectPositionWS()
  n15 = Abs(None=ComponentMask(None=Subtract(A=@n13.XYZ, B=@n14)) {b=true})
  n16 = Constant() {r=256}
  n17 = Power(Base=Divide(A=Clamp(None=@n15, Max=@n16), B=@n16), Exp=ScalarParameter() {defaultValue=0.75, parameterName=DecalHeightFalloffExp}) {constExponent=0.75}
  Multiply(A=DecalLifetimeOpacity().Opacity, B=Multiply(A=NamedRerouteUsage(), B=LinearInterpolate(A=@n12, Alpha=@n17) {constB=0}) {constA=1000})
  n18 = Constant() {r=1}
  n19 = ScalarParameter() {defaultValue=1, parameterName=Grunge0-1}
  n20 = TextureSample(UVs=@n2) {samplerSource=SSM_Wrap_WorldGroupSettings, samplerType=SAMPLERTYPE_Masks, texture=&T_BurnMark_Decal}
  n21 = Constant() {r=1}
  n22 = ScalarParameter() {defaultValue=1, parameterName=BurnMark0-1}
  n23 = TextureSample(UVs=Panner(Coordinate=@n5) {speedY=-0.03}) {
   samplerSource=SSM_Wrap_WorldGroupSettings
   samplerType=SAMPLERTYPE_LinearGrayscale
   texture=&Snowfall_pick_Noise
  }
  n24 = Add(A=@n23.R) {constB=0.5}
  n25 = Multiply(A=Multiply(A=SphereMask(A=@n20.R, B=@n21, Radius=@n22) {attenuationRadius=0.1, hardnessPercent=0}, B=Saturate(None=@n24))) {constB=2.5}
  NamedRerouteDeclaration(None=Add(A=SphereMask(A=@n3.B, B=@n18, Radius=@n19) {attenuationRadius=0.1, hardnessPercent=0}, B=Add(A=@n7, B=@n25))) {name=Opacity, nodeColor=<rgba 0 1 0.5312 1>}
  n26 = MaterialFunctionCall(UVs=TextureCoordinate(), Texture Scale=ScalarParameter() {defaultValue=1, parameterName=Uv Scale From Center}) {materialFunction=&ScaleUVsByCenter}
  NamedRerouteDeclaration(None=@n26.UVs) {name=ScaledUVs, nodeColor=<rgba 1 1 0 1>}
  NamedRerouteDeclaration(None=@n26.0-1 mask) {name=Scale Opac Mask, nodeColor=<rgba 0 1 0.9297 1>}
 ]
}
