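# Unlit, masked, two-sided master material for a rift gate's inner surface, faking a view into another place.
# Screen-space or virtual-plane UVs are pixelated, then distorted by noise and ripples to sample the interior with chromatic offset; an optional parallax transition blends two textures.
# Switchable sun, fog or star layers, time-of-day tint and edge glow form the color.
# An open-portal function drives vertex offset and dithered opacity, alongside camera offset and edge masks.
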Material /RiftGateBase/Gameplay/Materials/M_RiftGate_Inner_Master
graph {
 properties {
  bCastRayTracedShadows=false
  blendMode=BLEND_Masked
  pixelDepthOffsetMode=PDOM_Legacy
  shadingModel=MSM_Unlit
  twoSided=true
  usedShadingModels=Unlit
 }
 nodes [
  n1 = MaterialFunctionCall!(SphereRadius, SphereCenter) {materialFunction=/Game/Athena/Environments/Nexus/Pickles/Materials/MF_Pickle_Port_OpenWPO}
  NamedRerouteDeclaration(None=@n1.OpenAmount) {name=OpenAmount, nodeColor=<rgba 1 0 0.9656 1>}
  n2 = OneMinus(None=ComponentMask(None=TextureCoordinate()) {g=true})
  n3 = ScalarParameter() {defaultValue=8, parameterName=Edge Opacity G Mask Exp}
  n4 = Clamp(None=Multiply(A=Power(Base=@n2, Exp=@n3) {constExponent=5}, B=ScalarParameter() {defaultValue=10, parameterName=Edge Opacity G Mask Mult}) {constB=10})
  n5 = ComponentMask(None=TextureCoordinate()) {r=true}
  n6 = Constant2Vector() {g=0.5, r=0.5}
  n7 = ScalarParameter() {defaultValue=0.35, parameterName=Opacity Mask Radius R}
  n8 = ScalarParameter() {defaultValue=1, parameterName=Opacity Mask Hardness R}
  n9 = ComponentMask(None=TextureCoordinate()) {g=true}
  n10 = ScalarParameter() {defaultValue=0.35, parameterName=Opacity Mask Radius G}
  n11 = ScalarParameter() {defaultValue=0.7, parameterName=Opacity Mask Hardness G}
  n12 = Multiply(
   A=Min(
    A=SphereMask(A=@n5, B=@n6.R, Radius=@n7, Hardness=@n8) {hardnessPercent=50}
    B=SphereMask(A=@n6.G, B=@n9, Radius=@n10, Hardness=@n11) {hardnessPercent=50}
   )
   B=NamedRerouteUsage()
  )
  NamedRerouteDeclaration(None=MaterialFunctionCall(Alpha Threshold=Multiply(A=@n4, B=@n12)) {materialFunction=&DitherTemporalAA}.Result) {name=Edge Mask Opacity, nodeColor=<rgba 0 1 0.3672 1>}
  n13 = NamedRerouteUsage()
  n14 = Add(A=@n13, B=NamedRerouteUsage())
  NamedRerouteDeclaration(None=MaterialFunctionCall!(
    UV=StaticSwitchParameter(True=@n14, False=@n13) {defaultValue=true, parameterName=UseRipples}
    Center
    K1=ScalarParameter() {parameterName=Mustache UV}
    K2
    P1
    P2
    P3
    P4
    UV Scale Bias=ScalarParameter() {parameterName=Mustache UV Scale}
    Enable K2
    Mirror Tiling=StaticBoolParameter() {group=WIP, parameterName=MirrorTiling}
    Clamp Tiling=StaticBoolParameter() {defaultValue=true, group=WIP, parameterName=ClampTiling}
   ) {materialFunction=/Game/Global/MaterialFunctions/MF_MustacheDistortion}.Result) {name=Mustache UVs, nodeColor=<rgba 1 0.5156 0 1>}
  n15 = NamedRerouteUsage()
  n16 = TextureCoordinate()
  n17 = ComponentMask(None=@n16) {r=true}
  n18 = Constant() {r=0.5}
  n19 = OneMinus(None=Multiply(A=Distance(A=@n17, B=@n18)) {constB=2})
  n20 = Distance(A=ComponentMask(None=@n16) {g=true}, B=Constant())
  n21 = Multiply(A=Multiply(A=Sine(None=@n19) {period=3}, B=Multiply(A=@n20) {constB=2}), B=ScalarParameter() {defaultValue=50, parameterName=Stuff})
  n22 = Constant3Vector() {constant=<rgba 0 1 0 0>}
  NamedRerouteDeclaration(None=Multiply(A=Multiply(A=@n15, B=@n21), B=Transform(None=@n22) {transformSourceType=TRANSFORMSOURCE_Instance})) {name=Bubble WPO, nodeColor=<rgba 0.7109 1 0 1>}
  NamedRerouteDeclaration(None=@n1.DistanceFade) {name=Distance Fade, nodeColor=<rgba 0 1 0.0156 1>}
  NamedRerouteDeclaration(None=@n1.GlowAdd) {name=Add Glow, nodeColor=<rgba 1 0.0234 0 1>}
  NamedRerouteDeclaration(None=Add(A=@n1.PortalWPO, B=@n1.PortalWPO)) {name=Port OpenWPO, nodeColor=<rgba 0.6172 1 0 1>}
  n23 = Add(
   A=AppendVector(A=ScalarParameter() {group=WIP, parameterName=UVOffset_X}, B=ScalarParameter() {group=WIP, parameterName=UVOffset_Y})
   B=NamedRerouteUsage()
  )
  n24 = ScalarParameter() {parameterName=Virtual Plane Bump Offset}
  n25 = MaterialFunctionCall!(
   Heightmap Texture=TextureObjectParameter() {parameterName=Depth Mask, samplerType=SAMPLERTYPE_Masks, texture=&T_BlackMask}
   Height Ratio=@n24
   Min Steps=ScalarParameter() {defaultValue=2, parameterName=Min Steps Prallax}
   Max Steps=ScalarParameter() {defaultValue=8, parameterName=Max Steps Prallax}
   UVs=@n23
   Heightmap Channel=Constant4Vector() {constant=<rgba 1 0 0 0>}.RGBA
   Reference Plane=ScalarParameter() {parameterName=Virtual Plane Reference}
   -----------------------------
   Use World Coordinates
   Specify Manual Texture Size
   Manual Texture Size
   ------------------------------
   "Render Shadows (Occlusion Mapping)"
   Light Vector
   Shadow Steps
   Shadow Penumbra
   Transform To VertexNormal
  ) {materialFunction=/Engine/Functions/Engine_MaterialFunctions01/Texturing/ParallaxOcclusionMapping}
  n26 = QualitySwitch(
   Default=BumpOffset(Coordinate=@n23, Height=@n24, HeightRatioInput=ScalarParameter() {defaultValue=0.05, group=WIP, parameterName=HeightRatioInput}) {heightRatio=0.01}
   High=@n25.Parallax UVs
   Epic=@n25.Parallax UVs
  )
  n27 = Multiply(
   A=Multiply(A=@n26, B=ScalarParameter() {defaultValue=8, group=WIP, parameterName=Distortion_Tiling}) {constB=8}
   B=Constant2Vector() {g=0.5, r=1}
  )
  n28 = Multiply(A=Time(), B=ScalarParameter() {defaultValue=1, group=WIP, parameterName=DistSpeed})
  n29 = Add(A=@n27, B=@n28)
  n30 = Add(A=Multiply(A=@n27) {constB=0.8}, B=Multiply(A=@n28, B=Constant2Vector() {g=1, r=-1}))
  n31 = Multiply(A=Add(
    A=TextureSample(UVs=@n29) {texture=/Game/Effects/Fort_Effects/Textures/Tile/T_Smooth_Tile_Chromatic}.R
    B=TextureSample(UVs=@n30) {texture=&T_Smooth_Tile_Chromatic}.G
   )) {constB=0.5}
  n32 = ScalarParameter() {defaultValue=2, group=WIP, parameterName=distortpow}
  n33 = ScalarParameter() {defaultValue=0.25, group=WIP, parameterName=centX}
  n34 = ScalarParameter() {defaultValue=0.25, group=WIP, parameterName=centY}
  n35 = Constant2Vector() {g=1, r=0.5}
  n36 = OneMinus(None=SphereMask(
    A=AppendVector(A=@n33, B=@n34)
    B=MaterialFunctionCall(UVs=@n26, Texture Scale=@n35) {materialFunction=&ScaleUVsByCenter}.UVs
    Radius=ScalarParameter() {defaultValue=0.3, group=WIP, parameterName=rad}
    Hardness=ScalarParameter() {group=WIP, parameterName=hard}
   ) {attenuationRadius=0.2, hardnessPercent=0})
  n37 = ScalarParameter() {defaultValue=0.1, group=WIP, parameterName=distortbase}
  n38 = Multiply(A=Multiply(A=Power(Base=@n31, Exp=@n32), B=Max(A=@n36, B=@n37)), B=ScalarParameter() {group=WIP, parameterName=distortstrenght2})
  n39 = Multiply(A=@n36, B=ScalarParameter() {defaultValue=8, group=WIP, parameterName=MipFactor}) {constB=4}
  n40 = TextureSampleParameter2D!(UVs=Add(A=@n38, B=@n26), Level=@n39, Apply View MipBias) {mipValueMode=TMVM_MipLevel, parameterName=InteriorTextureA, texture=&T_BlacksRGB}
  n41 = TextureSampleParameter2D(UVs=Clamp(None=Add(A=@n38, B=@n26)) {maxDefault=0.99, minDefault=0.01}, Level=@n39) {mipValueMode=TMVM_MipLevel, parameterName=InteriorTextureB, texture=&T_BlacksRGB}
  n42 = ScalarParameter() {parameterName=InteriorBlend}
  n43 = VectorParameter() {parameterName=PlaneNormal, defaultValue=<rgba 0.1 0 -1 1>}
  n44 = Add(A=ActorPositionWS(), B=VectorParameter() {parameterName=WaterPlaneCenter, defaultValue=<rgba 0 0 1000 1>}.RGB)
  n45 = VectorParameter() {parameterName=WaterTile, defaultValue=<rgba 15000 15000 0 0>}
  n46 = MaterialFunctionCall(
   Plane Normal Axis=Normalize(VectorInput=Transform(None=@n43.RGB) {transformSourceType=TRANSFORMSOURCE_Local})
   Plane X Axis=Constant3Vector() {constant=<rgba 1 0 0 0>}
   Plane Center=@n44
   UV Scale=AppendVector(A=@n45.R, B=@n45.R)
  ) {materialFunction=&VirtualPlaneCoordinates}
  n47 = TextureSampleParameter2D!(UVs=Panner(Coordinate=@n46.UVs) {speedX=0.05, speedY=0.02}, Apply View MipBias) {parameterName=WaterNormal, samplerType=SAMPLERTYPE_Normal, texture=/Game/Athena/Apollo/Water/Textures/02_Rounded/02_WaterFoam_Veroni_base_N}
  n48 = TextureSampleParameter2D(UVs=Panner(Coordinate=Multiply(A=@n46.UVs) {constB=1.2}) {speedX=0.01, speedY=-0.02}) {parameterName=WaterNormal, samplerType=SAMPLERTYPE_Normal, texture=&02_WaterFoam_Veroni_base_N}
  n49 = ScalarParameter() {defaultValue=50000, parameterName=WaterNormalRadius}
  n50 = Multiply(
   A=Multiply(A=VectorParameter() {parameterName=WaterNormalMult, defaultValue=<rgba 0.5 0.5 1 1>}.RGB, B=LinearInterpolate(A=@n47.RGB, B=@n48.RGB))
   B=Power(Base=SphereMask(A=@n46.World Position, B=@n44, Radius=@n49) {hardnessPercent=0})
  )
  n51 = VertexColor()
  n52 = ScalarParameter() {defaultValue=1, parameterName=EdgeMaskPower}
  n53 = ScalarParameter() {parameterName=ExgeMaskPowerMin}
  n54 = MaterialFunctionCall() {materialFunction=&CameraDirectionVector}
  n55 = Transform(None=Constant3Vector() {constant=<rgba 0 1 0 0>}) {transformSourceType=TRANSFORMSOURCE_Instance}
  n56 = ScalarParameter() {defaultValue=0.9, parameterName=Virtual Plane Camera/Local Vector Blend}
  n57 = MaterialFunctionCall(
   Plane Normal Axis=Normalize(VectorInput=LinearInterpolate(A=@n54.Result, B=@n55, Alpha=@n56))
   Plane Center=TransformPosition(None=VectorParameter() {parameterName=Image Offset, defaultValue=<rgba 0 0 0 100>}.RGB) {transformSourceType=TRANSFORMPOSSOURCE_Instance, transformType=TRANSFORMPOSSOURCE_World}
   UV Scale=ScalarParameter() {defaultValue=200, parameterName=Image Size}
  ) {materialFunction=&VirtualPlaneCoordinates}
  n58 = ScalarParameter() {defaultValue=5000, parameterName=DepthOffset}
  n59 = LinearInterpolate(Alpha=Saturate(None=TwoSidedSign())) {constA=1, constB=-1}
  n60 = ScalarParameter() {defaultValue=1000, parameterName=HeightOffsetBase}
  n61 = ScalarParameter() {defaultValue=-1000, parameterName=HeightOffsetTop}
  n62 = ComponentMask(None=CameraPositionWS()) {b=true}
  n63 = ComponentMask(None=ActorPositionWS()) {b=true}
  n64 = Saturate(None=Divide(A=Subtract(A=@n62, B=@n63), B=ScalarParameter() {defaultValue=3000, parameterName=HeightOffsetMaskSize}))
  n65 = TransformPosition(None=MaterialFunctionCall(Y=Multiply(A=@n58, B=@n59), Z=LinearInterpolate(A=@n60, B=@n61, Alpha=@n64) {constB=0}) {materialFunction=&MakeFloat3}.Result) {transformType=TRANSFORMPOSSOURCE_World}
  n66 = Add(A=Subtract(
    A=ScreenPosition().ViewportUV
    B=MaterialFunctionCall(World Position=@n65) {materialFunction=&TransformToClipSpace}.Buffer UVs for Scene Textures
   )) {constB=0.5}
  n67 = ScalarParameter() {defaultValue=1200, parameterName=DistanceScale}
  n68 = Distance(A=ActorPositionWS(), B=CameraPositionWS())
  n69 = Constant() {r=1}
  n70 = MaterialFunctionCall(Float2=ViewSize()) {materialFunction=&BreakOutFloat2Components}
  n71 = Divide(A=@n70.R, B=@n70.R)
  n72 = ScalarParameter() {defaultValue=0.8, parameterName=ScaleMult}
  n73 = Multiply(A=NamedRerouteDeclaration(None=MaterialFunctionCall!(FlickerInitialSpeed, FlickerTimeOffset) {materialFunction=/Game/Athena/Environments/Nexus/Pickles/Materials/MF_Pickle_Port_Shudder}.Result) {name=ShudderResult, nodeColor=<rgba 0.4129 0.0452 1 1>}) {constB=-0.025}
  n74 = Multiply(
   A=Clamp(None=Divide(A=@n67, B=@n68), Max=ScalarParameter() {defaultValue=0.75, parameterName=MaxDistanceScaleUp}) {clampMode=CMODE_ClampMax, maxDefault=0}
   B=Multiply(A=AppendVector(A=@n69, B=@n71), B=Add(A=@n72, B=@n73))
  )
  n75 = MaterialFunctionCall!(
   UVs=StaticSwitchParameter(
    True=NamedRerouteDeclaration(None=@n57.UVs) {name=Virtual Plane, nodeColor=<rgba 1 0.7031 0 1>}
    False=MaterialFunctionCall(UVs=@n66, Texture Scale=@n74) {materialFunction=&ScaleUVsByCenter}.UVs
   ) {parameterName=Use Virtual Plane Coordionates}
   Resolution=Multiply(A=ScalarParameter() {defaultValue=100, parameterName=PixelRes}) {constB=0.5}
  ) {materialFunction=/Game/Athena/Environments/Nexus/Pickles/Materials/MF_Skirmish_Pixelate}
  n76 = Multiply(A=@n75.Result) {constB=3}
  n77 = TextureCoordinate() {uTiling=4, vTiling=2}
  n78 = Time()
  n79 = ScalarParameter() {defaultValue=2, parameterName=PanSpeedDistortion}
  n80 = TextureSample(UVs=Panner(Coordinate=LinearInterpolate(A=@n76, B=@n77) {constAlpha=0.6}, Time=Multiply(A=@n78, B=@n79)) {speedX=0.01, speedY=-0.07}) {texture=&T_Smooth_Tile_Chromatic}
  n81 = Constant2Vector() {g=1, r=0.05}
  n82 = VectorParameter() {parameterName=DistortionTile, defaultValue=<rgba 0.3 0.1 2 0>}
  n83 = AppendVector(A=@n82.R, B=@n82.R)
  n84 = Add(A=Multiply(A=@n77, B=Multiply(A=@n81, B=@n83)), B=Constant2Vector() {r=1.75})
  n85 = Multiply(A=@n77, B=@n83)
  n86 = NamedRerouteUsage()
  n87 = Time()
  n88 = ScalarParameter() {defaultValue=0.1, parameterName=PanSpeed}
  n89 = TextureSampleParameter2D(UVs=Panner(Coordinate=LinearInterpolate(A=@n84, B=@n85, Alpha=@n86) {constA=4}, Time=Multiply(A=@n87, B=@n88)) {speedY=-0.2}) {parameterName=DistortionTexture, texture=/Game/Effects/Fort_Effects/Textures/Tile/T_Caustics_Chromatic}
  n90 = Desaturation(None=LinearInterpolate(A=@n80.RGB, B=@n89.RGB))
  n91 = NamedRerouteUsage()
  n92 = Add(
   A=Power(Base=OneMinus(None=@n51.G), Exp=LinearInterpolate(A=@n52, B=@n53, Alpha=@n90) {constA=3, constB=0.25}) {constExponent=0.5}
   B=Multiply(A=@n91) {constB=0.2}
  )
  n93 = SphereMask(
   A=ComponentMask(None=TextureCoordinate()) {r=true}
   B=Constant() {r=0.5}
   Radius=ScalarParameter() {defaultValue=0.5, parameterName=InteriorLinesRadius}
  ) {attenuationRadius=500, hardnessPercent=0}
  n94 = LinearInterpolate(Alpha=@n93) {constA=1.6, constB=-0.4}
  n95 = Power(Base=@n93, Exp=ScalarParameter() {defaultValue=3, parameterName=InteriorLinesPower})
  n96 = ComponentMask(None=TextureCoordinate()) {g=true}
  n97 = VectorParameter() {parameterName=LinesTile, defaultValue=<rgba 5 3 0 0>}
  n98 = Multiply(A=@n96, B=@n97.R)
  n99 = Panner(Coordinate=Multiply(A=AppendVector(A=@n95, B=@n96), B=AppendVector(A=@n98, B=@n97.G)) {constB=2}) {speedX=-0.1, speedY=-0.17}
  n100 = QualitySwitch(
   Default=Multiply(A=Saturate(None=@n94), B=TextureSample(UVs=@n99) {texture=/Game/Athena/Environments/Nexus/Textures/T_ZeroPoint_MagneticWaves}.RGB) {constB=0}
   Low=Constant()
  )
  n101 = ComponentMask(None=@n100) {g=true, r=true}
  n102 = ScalarParameter() {defaultValue=0.3, parameterName=EdgeRippleDistortion}
  n103 = NamedRerouteDeclaration(None=Multiply(A=@n92, B=Multiply(A=@n101, B=@n102))) {name=Ripples, nodeColor=<rgba 0 1 0.8359 1>}
  n104 = Add(
   A=@n75.Result
   B=Multiply(A=ComponentMask(None=@n80.RGB) {g=true, r=true}, B=ScalarParameter() {defaultValue=0.017, parameterName=MainDistortion})
  ) {constB=0}
  n105 = NamedRerouteDeclaration(None=StaticSwitchParameter(True=Add(A=@n104, B=@n103) {constB=0}, False=@n104) {defaultValue=true, parameterName=UseRipples}) {name=Distorted UVs, nodeColor=<rgba 0.0078 0 1 1>}
  n106 = Add(A=ShadingPathSwitch(Default=ComponentMask(None=@n50) {g=true, r=true}, Mobile=Constant()), B=@n105)
  n107 = TextureCoordinate()
  n108 = Normalize(VectorInput=AppendVector(A=Subtract(A=@n107) {constB=0.5}, B=Constant()))
  n109 = Desaturation(None=Multiply(A=@n100), Fraction=Constant() {r=-1})
  n110 = Add(A=@n80.RGB, B=StaticSwitchParameter(True=Add(A=@n89.RGB, B=@n109), False=@n89.RGB) {defaultValue=true, parameterName=UseRipples})
  n111 = Add(A=@n110, B=@n100)
  n112 = Multiply(A=@n108, B=Multiply(A=ScalarParameter() {defaultValue=0.005, parameterName=ChromaticAmount}, B=Add(A=@n111) {constB=1.5})) {constB=2}
  n113 = MaterialFunctionCall(Float3=Multiply(A=@n92, B=StaticSwitchParameter(True=@n112, False=@n108) {defaultValue=true, parameterName=UseRipples})) {desc=Chromatic offset, materialFunction=&BreakOutFloat3Components}
  n114 = Add(A=@n105, B=@n113.R)
  n115 = Add(A=@n105, B=@n113.G)
  n116 = TextureSampleParameter2D(UVs=@n105) {parameterName=InteriorTexture, texture=/RiftGateBase/Gameplay/Textures/T_Riftgate_BG}
  n117 = StaticSwitchParameter(
   True=TextureSampleParameter2D(UVs=NamedRerouteDeclaration(None=@n106) {name=InteriorTextureUV, nodeColor=<rgba 1 0 0.75 1>}) {parameterName=InteriorTexture, texture=&T_Riftgate_BG}.RGB
   False=MaterialFunctionCall(
    R=TextureSampleParameter2D(UVs=@n114) {parameterName=InteriorTexture, texture=&T_Riftgate_BG}.R
    G=TextureSampleParameter2D(UVs=@n115) {parameterName=InteriorTexture, texture=&T_Riftgate_BG}.G
    B=@n116.B
   ) {materialFunction=&AppendMany}.RGB
  ) {parameterName=Dusty}
  n118 = ShadingPathSwitch(Default=@n117, Forward=@n116.RGB, Mobile=@n116.RGB)
  NamedRerouteDeclaration(None=StaticSwitchParameter(True=LinearInterpolate(A=@n40.RGB, B=@n41.RGB, Alpha=@n42), False=QualitySwitch(Default=@n118, Low=@n116.RGB, Medium=@n117)) {parameterName=UseInteriorTextureTransition}) {name=InteriorTexture, nodeColor=<rgba 0.0781 1 0 1>}
  n119 = VectorParameter() {parameterName=Sunoffset, defaultValue=<rgba 0.21 0.35 0.5 0>}
  n120 = Add(A=Multiply(A=@n104, B=@n119.B), B=AppendVector(A=@n119.R, B=@n119.R))
  n121 = Rotator(Coordinate=@n120)
  n122 = Rotator(Coordinate=@n120) {speed=-0.2}
  n123 = Multiply(
   A=Multiply(
    A=TextureSampleParameter2D(UVs=@n121) {parameterName=SunTexture, texture=/Game/Effects/Fort_Effects/Textures/Flares/T_Flare_Streaks}.RGB
    B=TextureSampleParameter2D(UVs=@n122) {parameterName=SunTexture, texture=&T_Flare_Streaks}.RGB
   )
   B=VectorParameter() {parameterName=SunBrightness, defaultValue=<rgba 0.2 0.2 0.2 0>}.RGB
  ) {constB=-1}
  n124 = TextureSampleParameter2D(UVs=@n104) {parameterName=DustyMask, texture=&T_BlacksRGB}
  n125 = NamedRerouteUsage()
  n126 = ScalarParameter() {parameterName=WombatFogMaskOffset}
  n127 = Multiply(A=ScalarParameter() {defaultValue=1, parameterName=WombatFogMaskSize}, B=Saturate(None=ComponentMask(None=@n104) {g=true}))
  n128 = Saturate(None=Sine(None=Add(A=@n126, B=@n127)) {period=0.5})
  n129 = VectorParameter() {parameterName=WombatFogTile, defaultValue=<rgba 1.8 3.5 5 0>}
  n130 = AppendVector(A=@n129.R, B=@n129.R)
  n131 = VectorParameter() {parameterName=WombatFogPanDirection, defaultValue=<rgba 0.02 0.12 0.15 0>}
  n132 = Multiply(A=Time(), B=@n131.B)
  n133 = Panner(
   Coordinate=MaterialFunctionCall(XY=Multiply(A=@n130, B=@n104) {constB=2}) {materialFunction=&Swizzle}.YX
   Time=@n132
   Speed=AppendVector(A=@n131.R, B=@n131.R)
  ) {speedX=0.02, speedY=0.15}
  n134 = TextureSample(UVs=Panner(Coordinate=Multiply(A=@n104, B=@n129.B) {constB=5}, Time=@n132) {speedX=-0.08, speedY=-0.04}) {texture=&T_Smooth_tile}
  n135 = VectorParameter() {parameterName=WombatFogColor, defaultValue=<rgba 1 1 1 0>}
  n136 = Multiply(
   A=Multiply(
    A=TextureSampleParameter2D(UVs=@n133) {parameterName=WombatFog, texture=/Game/Packages/Fortress_Sky/SkyDome/Master/T_CloudWhisp_Pattern_04}.G
    B=LinearInterpolate(Alpha=@n134.G) {constA=0.3, constB=1.2}
   )
   B=@n135.RGB
  )
  n137 = Saturate(None=MaterialFunctionCall!(FlickerInitialSpeed=ScalarParameter() {defaultValue=0.7, parameterName=FlickerSpeed2}, FlickerRandom=Constant()) {materialFunction=/Game/Effects/Fort_Effects/Materials/_Functions/MF_Flicker}.Result)
  n138 = Panner(Coordinate=NamedRerouteUsage()) {speedX=0.2}
  n139 = Multiply(
   A=Add(
    A=TextureSampleParameter2D(UVs=@n138) {parameterName=FlickerTextureMask, texture=&T_BlacksRGB}.R
    B=Time()
   )
   B=ScalarParameter() {defaultValue=1, parameterName=FlickerSpeed}
  )
  n140 = Sine(None=@n139) {period=0.52}
  n141 = Add(A=Add(A=Sine(None=@n139) {period=0.81}, B=Sine(None=@n139) {period=3.5})) {constB=1.8}
  n142 = Saturate(None=Multiply(A=Add(A=@n140, B=@n141)) {constB=3})
  n143 = StaticSwitchParameter(True=Clamp(None=LinearInterpolate(A=@n137, Alpha=@n142), Min=@n135.A) {clampMode=CMODE_ClampMin}, False=Constant() {r=1}) {parameterName=Typhoon}
  n144 = QualitySwitch(Default=Multiply(A=Power(Base=@n128), B=Multiply(A=@n136, B=@n143)), Low=Constant())
  n145 = TextureSample(UVs=@n104) {samplerType=SAMPLERTYPE_Grayscale, texture=/Game/Athena/Environments/Nexus/Pickles/Textures/T_Pickle_Int_01_M}
  n146 = Time()
  n147 = ScalarParameter() {defaultValue=0.25, parameterName=CherryAnimSpeed}
  n148 = Sine(None=Add(A=@n134.G, B=Multiply(A=@n146, B=@n147)))
  n149 = Multiply(
   A=VectorParameter() {parameterName=BrightStarColor, defaultValue=<rgba 25 20.9249 10.8103 0>}.RGB
   B=Add(A=@n145.R, B=Multiply(A=@n145.R, B=@n148))
  ) {constB=15}
  n150 = Add(A=@n125) {constB=0.1}
  n151 = StaticSwitchParameter(True=Add(A=Multiply(A=@n149, B=@n150), B=Multiply(A=@n125) {constB=0.4}), False=@n125) {parameterName=Cherry}
  n152 = StaticSwitchParameter(
   True=Add(A=@n125, B=Multiply(A=@n123, B=@n124.G))
   False=StaticSwitchParameter(True=Add(A=@n144, B=@n151), False=@n151) {parameterName=Wombat}
  ) {parameterName=Dusty}
  n153 = MaterialFunctionCall(
   Color_Morning=VectorParameter() {parameterName=MorningTint, defaultValue=<rgba 0.4 0.3098 0.2735 0>}.RGB
   Color_Day=VectorParameter() {parameterName=DayTint, defaultValue=<rgba 1 1 1 0>}.RGB
   Color_Evening=VectorParameter() {parameterName=EveningTint, defaultValue=<rgba 0.5 0.4215 0.4165 0>}.RGB
   Color_Night=VectorParameter() {parameterName=NightTint, defaultValue=<rgba 1.5 1.3961 0.9576 0>}.RGB
  ) {materialFunction=&Color_Time_of_Day}
  n154 = Multiply(
   A=LinearInterpolate(
    A=VectorParameter() {parameterName=EdgeColor, defaultValue=<rgba 1.1639 0.8936 2.5 0>}.RGB
    B=VectorParameter() {parameterName=EdgeColorDestabilize, defaultValue=<rgba 0.868 0.3313 1 0>}.RGB
    Alpha=@n91
   )
   B=Add(A=Multiply(A=@n91, B=@n110), B=@n110)
  )
  n155 = Multiply(A=@n92, B=@n92)
  n156 = Multiply(A=Add(A=Multiply(A=@n152, B=@n153.Result), B=Multiply(A=@n154, B=@n155)), B=NamedRerouteUsage())
  n157 = TwoSidedSign()
  StaticSwitchParameter(True=LinearInterpolate(A=@n156, B=NamedRerouteUsage(), Alpha=Saturate(None=@n157)), False=@n156) {parameterName=UseBubbleWPO}
  n158 = Add(
   A=NamedRerouteUsage()
   B=MaterialFunctionCall(Offset Amount=ScalarParameter() {defaultValue=-100, parameterName=CamOffset}) {materialFunction=/Engine/Functions/Engine_MaterialFunctions02/WorldPositionOffset/CameraOffset}.Camera Offset
  )
  n159 = StaticSwitchParameter(True=Add(A=@n158, B=NamedRerouteUsage()), False=@n158) {parameterName=UseBubbleWPO}
  n160 = MaterialFunctionCall(Y=Multiply(A=VertexColor().R, B=ScalarParameter() {defaultValue=100, group=WIP, parameterName=Thickness}) {constB=50}) {materialFunction=&MakeFloat3}
  StaticSwitchParameter(True=Add(A=@n159, B=Transform(None=@n160.Result) {transformSourceType=TRANSFORMSOURCE_Local}), False=@n159) {group=WIP, parameterName=UseThickness}
  n161 = NamedRerouteUsage()
  n162 = Power(Base=@n1.OpenAmount) {constExponent=0.5}
  n163 = ScalarParameter() {parameterName=VertColorAlphaContrast}
  n164 = Multiply(
   A=Multiply(
    A=NamedRerouteDeclaration(None=@n162) {name=Open Amount Opacity, nodeColor=<rgba 1 0.4922 0 1>}
    B=MaterialFunctionCall(In=@n51.G, Contrast=@n163) {materialFunction=&CheapContrast}.Result
   )
   B=ScalarParameter() {defaultValue=1, parameterName=OpacityMult}
  )
  n165 = SphereMask(
   A=@n104
   B=Constant() {r=0.5}
   Radius=ScalarParameter() {defaultValue=0.65, parameterName=OpacitySphereRadius}
   Hardness=ScalarParameter() {defaultValue=0.3, parameterName=OpacitySphereHardness}
  ) {hardnessPercent=0.7}
  n166 = MaterialFunctionCall(Alpha Threshold=Multiply(A=@n1.WeldFade, B=Multiply(A=@n164, B=@n165))) {materialFunction=&DitherTemporalAA}
  n167 = StaticSwitchParameter(
   True=NamedRerouteUsage()
   False=StaticSwitchParameter(True=Multiply(A=@n161, B=@n166.Result), False=@n166.Result) {parameterName=UseDistanceFade}
  ) {parameterName=Use Mesh Edge Mask Opacity}
  n168 = Abs(None=Subtract(A=TwoSidedSign()))
  StaticSwitchParameter(True=Multiply(A=@n167, B=Saturate(None=@n168)), False=@n167) {group=WIP, parameterName=HidePositiveSign}
 ]
}
