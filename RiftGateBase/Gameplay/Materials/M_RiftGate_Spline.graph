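# Translucent, unlit, two-sided Rift Gate material for spline meshes and Niagara mesh particles, rendering flowing, starry energy strands.
# World position offset combines spline thickening, a travelling sine wave, optional thickness and a portal-opening function, varied by vertex colors.
# Emissive layers two panning star textures, distorted by chromatic caustics, with tints, background gradient and time-of-day scaling.
# Opacity derives from those textures, end masks and optional near-camera and depth fades.
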Material /RiftGateBase/Gameplay/Materials/M_RiftGate_Spline
graph {
 properties {
  bUsedWithNiagaraMeshParticles=true
  bUsedWithSplineMeshes=true
  blendMode=BLEND_Translucent
  ditherOpacityMask=true
  pixelDepthOffsetMode=PDOM_Legacy
  shadingModel=MSM_Unlit
  twoSided=true
  usedShadingModels=Unlit
 }
 nodes [
  n1 = ScalarParameter() {defaultValue=200, group=Spline Thicken, parameterName=SplineWidth}
  LinearInterpolate(B=@n1, Alpha=NamedRerouteUsage()) {constA=50, constB=0}
  n2 = OneMinus(None=ComponentMask(None=TextureCoordinate()) {g=true})
  n3 = ScalarParameter() {group=Spawn In, parameterName=SpawnInfluence}
  n4 = LinearInterpolate(Alpha=@n3) {constA=0.4, constB=-1.1}
  n5 = LinearInterpolate(Alpha=@n3) {constA=1.2, constB=0.5}
  NamedRerouteDeclaration(None=Clamp(None=SphereMask(A=@n2, B=@n4, Radius=@n5) {attenuationRadius=1, hardnessPercent=50})) {name=SpawnIN, nodeColor=<rgba 0.1719 1 0 1>}
  n6 = VertexColor()
  n7 = MaterialFunctionCall!(
   SphereRadius=LinearInterpolate(
    A=ScalarParameter() {defaultValue=0.6, group=Portal Spheres WPO, parameterName=SphereRadiusOuter}
    B=ScalarParameter() {defaultValue=0.5, group=Portal Spheres WPO, parameterName=SphereRadiusInner}
    Alpha=@n6.R
   )
   SphereCenter=LinearInterpolate(
    A=ScalarParameter() {defaultValue=0.6, group=Portal Spheres WPO, parameterName=SphereCenterOuter}
    B=ScalarParameter() {defaultValue=0.5, group=Portal Spheres WPO, parameterName=SphereCenterInner}
    Alpha=@n6.R
   )
  ) {materialFunction=/Game/Athena/Environments/Nexus/Pickles/Materials/MF_Pickle_Port_OpenWPO}
  NamedRerouteDeclaration(None=@n7.OpenAmount) {name=OpenAmount, nodeColor=<rgba 0.6875 0 1 1>}
  NamedRerouteDeclaration(None=@n7.DistanceFade) {name=DistanceFade, nodeColor=<rgba 0.2187 1 0 1>}
  NamedRerouteDeclaration(None=@n7.PortalWPO) {name=PortalWPO, nodeColor=<rgba 1 0.7031 0 1>}
  n8 = ComponentMask(None=TextureCoordinate() {coordinateIndex=1}) {g=true}
  n9 = Add(
   A=Multiply(A=Time(), B=ScalarParameter() {defaultValue=-0.15, group=Sine Wave, parameterName=WaveSpeed})
   B=Divide(A=@n8, B=ScalarParameter() {defaultValue=0.75, group=Sine Wave, parameterName=WaveLength})
  )
  n10 = Add(A=VertexColor().G, B=@n9)
  NamedRerouteDeclaration(None=ConstantBiasScale(None=Sine(None=@n10))) {name=VertexColorSine, nodeColor=<rgba 0 0.3437 1 1>}
  n11 = VertexColor()
  n12 = Multiply(A=ScalarParameter() {defaultValue=100, group=WIP, parameterName=Thickness}, B=OneMinus(None=NamedRerouteUsage()))
  n13 = Transform(None=MaterialFunctionCall(Y=Multiply(A=@n11.R, B=@n12) {constB=50}) {materialFunction=&MakeFloat3}.Result) {transformSourceType=TRANSFORMSOURCE_Local}
  n14 = LinearInterpolate(Alpha=NamedRerouteUsage()) {constA=0.5}
  n15 = ScalarParameter() {defaultValue=1, group=Spline Thicken, parameterName=BaseUVTile}
  n16 = VertexColor()
  n17 = Multiply(A=Multiply(A=@n1, B=StaticSwitchParameter(True=@n14, False=@n15) {defaultValue=true, group=WIP, parameterName=UseUVDistanceBlend}), B=@n16.B)
  n18 = ScalarParameter() {defaultValue=0.25, group=Spline Thicken, parameterName=WidthTipMult}
  n19 = NamedRerouteUsage()
  n20 = TextureCoordinate() {coordinateIndex=1}
  n21 = ScalarParameter() {defaultValue=0.5, group=End Fade, parameterName=EndFadeRadius_Y}
  n22 = SphereMask(
   A=ComponentMask(None=@n20) {g=true}
   B=@n21
   Radius=@n21
   Hardness=ScalarParameter() {defaultValue=0.5, group=End Fade, parameterName=EndFadeHardness_Y}
  ) {hardnessPercent=0.25}
  n23 = MaterialFunctionCall(
   WidthBase=LinearInterpolate(A=Multiply(A=@n17, B=@n18), Alpha=@n19) {constB=0}
   WidthTip=LinearInterpolate(A=NamedRerouteDeclaration(None=@n17) {name=SplineWidth, nodeColor=<rgba 0.4062 0 1 1>}, Alpha=@n19) {constB=0}
   UVs for Projection=TextureCoordinate() {coordinateIndex=1}
   UVs for Thickness=OneMinus(None=@n22)
  ) {materialFunction=&SplineThicken}
  n24 = MaterialFunctionCall() {materialFunction=&TangentBasis}
  n25 = NamedRerouteUsage()
  n26 = Add(A=Add(A=OneMinus(None=@n16.B)), B=LinearInterpolate(Alpha=NamedRerouteUsage()) {constA=1, constB=0.4})
  n27 = Saturate(None=LinearInterpolate(Alpha=@n8) {constA=5, constB=0.2})
  n28 = Multiply(
   A=Multiply(A=LinearInterpolate(A=@n24.X U, B=@n24.X U, Alpha=@n25), B=Sine(None=@n9))
   B=Multiply(A=Multiply(A=@n26, B=@n27), B=ScalarParameter() {defaultValue=50, group=Sine Wave, parameterName=WaveScale})
  )
  n29 = Add(
   A=StaticSwitchParameter(True=Add(A=@n23.WorldPositionOffset, B=@n28), False=@n28) {defaultValue=true, parameterName=UseSplineThicken}
   B=Multiply(A=NamedRerouteUsage(), B=ScalarParameter() {defaultValue=-1, group=Spline Thicken, parameterName=WPOMUlt}) {constB=-1}
  )
  Reroute(None=StaticSwitchParameter(True=Add(A=@n13, B=@n29), False=@n29) {group=WIP, parameterName=UseThickness})
  n30 = VectorParameter() {group=Texture Parameter, parameterName=Tile1, defaultValue=<rgba 0.15 8 0 0>}
  n31 = Multiply(A=ScalarParameter() {defaultValue=0.02, parameterName=SplineWidthTileMult}, B=NamedRerouteUsage()) {constB=0.1}
  n32 = Multiply(A=TextureCoordinate() {coordinateIndex=1}, B=AppendVector(A=Multiply(A=@n30.R, B=@n31), B=@n30.G))
  n33 = Subtract(A=TextureCoordinate()) {constB=0.5}
  n34 = OneMinus(None=Multiply(A=Abs(None=@n33)) {constB=2})
  n35 = LinearInterpolate(
   A=ScalarParameter() {defaultValue=1, group=Texture Parameter, parameterName=TilingScale_Edges}
   B=ScalarParameter() {defaultValue=0.25, group=Texture Parameter, parameterName=TilingScale_Center}
   Alpha=ComponentMask(None=NamedRerouteDeclaration(None=@n34) {name=CenterMask, nodeColor=<rgba 1 0.5625 0 1>}) {r=true}
  ) {constA=1, constB=0.25}
  n36 = Constant() {r=1}
  n37 = StaticSwitchParameter(
   True=MaterialFunctionCall(UVs=@n32, Texture Scale=AppendVector(A=@n35, B=@n36)) {materialFunction=&ScaleUVsByCenter}.UVs
   False=@n32
  ) {group=Texture Parameter, parameterName=UseCenterUVScaling}
  n38 = VectorParameter() {group=Texture Parameter, parameterName=Tile2, defaultValue=<rgba 1 0.1 0 0>}
  n39 = AppendVector(A=@n38.R, B=@n38.R)
  n40 = VectorParameter() {group=Texture Parameter, parameterName=PanDirection2, defaultValue=<rgba 0.04 0.1 0 0>}
  n41 = VertexColor()
  n42 = Time()
  n43 = Add(A=@n41.G, B=Multiply(A=Multiply(A=@n41.B, B=@n42), B=ScalarParameter() {defaultValue=-0.5, parameterName=PanSpeed}))
  n44 = TextureSample(UVs=Panner(Coordinate=Multiply(A=@n37, B=@n39), Time=@n43, Speed=AppendVector(A=@n40.R, B=@n40.R)) {speedX=1, speedY=1}) {texture=/Game/Effects/Fort_Effects/Textures/Tile/T_Caustics_Chromatic}
  n45 = Multiply(
   A=ComponentMask(None=@n44.RGB) {g=true, r=true}
   B=ScalarParameter() {defaultValue=0.05, group=Texture Parameter, parameterName=DistortionIntensity}
  ) {constB=0.05}
  n46 = VectorParameter() {group=Texture Parameter, parameterName=PanDirection1, defaultValue=<rgba 0.08 1 0 0>}
  n47 = AppendVector(A=@n46.R, B=@n46.R)
  n48 = TextureSample(UVs=Panner(Coordinate=Add(A=@n37, B=@n45), Time=@n43, Speed=@n47) {speedX=1, speedY=1}) {texture=/Game/Athena/Environments/Nexus/Textures/T_ZeroPoint_Line_Stars}
  n49 = Multiply(A=VectorParameter() {group=Color, parameterName=PrimaryTint, defaultValue=<rgba 1 1 1 1>}.RGB, B=@n48.RGB) {constB=0.5}
  n50 = Multiply(A=@n37) {constB=2}
  n51 = VectorParameter() {group=Texture Parameter, parameterName=PanDirection1Mult, defaultValue=<rgba -1.7 2.4 0 1>}
  n52 = AppendVector(A=@n51.R, B=@n51.R)
  n53 = TextureSample(UVs=Panner(Coordinate=Subtract(A=@n50, B=@n45), Time=@n43, Speed=Multiply(A=@n47, B=@n52)) {speedX=1, speedY=1}) {texture=&T_ZeroPoint_Line_Stars}
  n54 = Multiply(A=@n53.RGB, B=VectorParameter() {group=Color, parameterName=SecondaryTint}.RGB) {constB=0.5}
  n55 = VectorParameter() {group=Color, parameterName=BackgroundColor1, defaultValue=<rgba 0.0396 0.1857 0.2 0>}
  n56 = VectorParameter() {group=Color, parameterName=BackgroundColor2, defaultValue=<rgba 0.0438 0.0296 0.15 0>}
  n57 = NamedRerouteUsage()
  n58 = ScalarParameter() {defaultValue=0.5, group=End Fade, parameterName=EndFadeRadius_X}
  n59 = Multiply(A=SphereMask(A=ComponentMask(None=@n20) {r=true}, B=@n58, Radius=@n58, Hardness=ScalarParameter() {group=End Fade, parameterName=EndFadeHardness_X}) {hardnessPercent=0})
  n60 = Multiply(
   A=MaterialFunctionCall(In=@n59, Contrast=Constant() {r=0.25}) {materialFunction=&CheapContrast}.Result
   B=VertexColor().B
  )
  n61 = Add(A=Add(A=Add(A=@n49, B=@n54), B=@n44.RGB), B=Multiply(A=LinearInterpolate(A=@n55.RGB, B=@n56.RGB, Alpha=@n57), B=@n60))
  n62 = NamedRerouteUsage()
  n63 = ComponentMask(None=@n62) {r=true}
  n64 = ComponentMask(None=@n62) {g=true}
  n65 = LinearInterpolate(
   A=@n61
   B=VectorParameter() {group=Color, parameterName=CenterColor, sortPriority=41, defaultValue=<rgba 0.1615 0.4157 1 1>}.RGB
   Alpha=Power(Base=Min(A=@n63, B=@n64), Exp=ScalarParameter() {defaultValue=1, group=Color, parameterName=CenterPower, sortPriority=42})
  )
  n66 = MaterialFunctionCall() {materialFunction=/Game/Athena/Environments/Nexus/Pickles/Materials/MF_Pickle_Tendril_Colors}
  n67 = Multiply(
   A=VectorParameter() {group=Color, parameterName=EmissiveMult, defaultValue=<rgba 0.7893 0.75 1 0>}.RGB
   B=MaterialFunctionCall(
    Value_Morning=ScalarParameter() {defaultValue=1, group=Time Of Day, parameterName=Morning Scalar}
    Value_Day=ScalarParameter() {defaultValue=1, group=Time Of Day, parameterName=Day Scalar}
    Value_Evening=ScalarParameter() {defaultValue=1, group=Time Of Day, parameterName=Evening Scalar}
    Value_Night=ScalarParameter() {defaultValue=1, group=Time Of Day, parameterName=Night Scalar}
   ) {materialFunction=&Scalar_Time_of_Day}.Result
  )
  Reroute(None=Multiply(
    A=StaticSwitchParameter(True=@n65, False=@n61) {group=Color, parameterName=UseCenterColorBlend, sortPriority=40}
    B=Multiply(A=@n66.Result, B=@n67)
   ))
  n68 = Abs(None=Subtract(A=TwoSidedSign()))
  n69 = NamedRerouteUsage()
  n70 = Add(A=@n48.B, B=Add(A=@n53.B, B=@n44.B))
  n71 = Multiply(A=Multiply(A=Add(A=@n70, B=@n60) {constB=0.5}, B=@n59), B=@n22)
  n72 = ScalarParameter() {defaultValue=800, group=Opacity, parameterName=NearFadeDIst}
  n73 = ScalarParameter() {defaultValue=70, group=Opacity, parameterName=DepthFade}
  n74 = Constant() {r=0.7}
  n75 = ShadingPathSwitch(
   Default=Multiply(
    A=MaterialFunctionCall(Distance=@n72) {materialFunction=&Near_Camera_Fade}.Result
    B=DepthFade(FadeDistance=@n73)
   )
   Forward=@n74
   Mobile=@n74
  )
  n76 = Multiply(
   A=StaticSwitchParameter(True=Multiply(A=@n71, B=@n75), False=@n71) {defaultValue=true, parameterName=AllowDepthFade}
   B=ScalarParameter() {defaultValue=1, group=Opacity, parameterName=OpacityMult}
  )
  n77 = StaticSwitchParameter(True=@n76, False=Multiply(A=@n76, B=OneMinus(None=@n69))) {defaultValue=true, parameterName=UseSplineThicken}
  StaticSwitchParameter(True=Multiply(A=@n77, B=Saturate(None=@n68)), False=@n77) {group=WIP, parameterName=HidePositiveSign}
 ]
}
